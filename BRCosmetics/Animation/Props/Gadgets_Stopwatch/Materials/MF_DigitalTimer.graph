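# Material function that draws a stopwatch-style digital timer mask from input UVs, showing minutes, seconds and hundredths.
# Time comes from material Time minus TimeStart, zeroed when TimeStart is zero, or from OverrideTime via the UseTimestamping static switch.
# Each field feeds MF_NumericDigit with parameter-driven scale and offset.
# SphereMask dots form separators, the colon mirrored across V from one mask, and Max merges everything into one output.
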MaterialFunction /BRCosmetics/Animation/Props/Gadgets_Stopwatch/Materials/MF_DigitalTimer
graph {nodes [
  n1 = VectorParameter() {
   desc="R = Scale X, G = Scale Y, B = Offset X, A = Offset Y"
   group=DigitalTimer
   parameterName=Hundredth of a Second Scale/Offset
   sortPriority=3
   channelNames {a=Offset Y, b=Offset X, g=Scale Y, r=Scale X}
   defaultValue=<rgba 2 0.3 -0.16 0>
  }
  n2 = MaterialFunctionCall(UVs=NamedRerouteUsage(), Texture Scale=AppendVector(A=@n1.Scale X, B=@n1.Scale X)) {materialFunction=&ScaleUVsByCenter}
  n3 = AppendVector(A=@n1.Offset X, B=@n1.Offset X)
  n4 = Multiply(A=Frac(None=NamedRerouteUsage())) {constB=100}
  n5 = VectorParameter() {
   desc="R = Scale X, G = Scale Y, B = Offset X, A = Offset Y"
   group=DigitalTimer
   parameterName=Seconds Scale/Offset
   sortPriority=2
   channelNames {a=Offset Y, b=Offset X, g=Scale Y, r=Scale X}
   defaultValue=<rgba 2 0.3 0 0>
  }
  n6 = MaterialFunctionCall(UVs=NamedRerouteUsage(), Texture Scale=AppendVector(A=@n5.Scale X, B=@n5.Scale X)) {materialFunction=&ScaleUVsByCenter}
  n7 = AppendVector(A=@n5.Offset X, B=@n5.Offset X)
  n8 = Fmod(A=NamedRerouteUsage(), B=Constant() {r=60})
  n9 = Frac(None=@n8)
  n10 = Max(
   A=MaterialFunctionCall!(UVs=Add(A=@n2.UVs, B=@n3), Digits=Ceil(None=@n4)) {materialFunction=&/MF_NumericDigit}.Result
   B=MaterialFunctionCall(UVs=Add(A=@n6.UVs, B=@n7), Digits=Subtract(A=@n8, B=@n9)) {materialFunction=&MF_NumericDigit}.Result
  )
  n11 = NamedRerouteUsage()
  n12 = VectorParameter() {
   desc="R = Scale X, G = Scale Y, B = Offset X, A = Offset Y"
   group=DigitalTimer
   parameterName=Minutes Scale/Offset
   sortPriority=1
   channelNames {a=Offset Y, b=Offset X, g=Scale Y, r=Scale X}
   defaultValue=<rgba 2 0.3 0.16 0>
  }
  n13 = AppendVector(A=@n12.Scale X, B=@n12.Scale X)
  n14 = Constant() {r=60}
  n15 = Fmod(A=Divide(A=NamedRerouteUsage(), B=@n14) {constB=60}, B=@n14)
  n16 = MaterialFunctionCall(
   UVs=Add(
    A=MaterialFunctionCall(UVs=@n11, Texture Scale=@n13) {materialFunction=&ScaleUVsByCenter}.UVs
    B=AppendVector(A=@n12.Offset X, B=@n12.Offset X)
   )
   Digits=Subtract(A=@n15, B=Frac(None=@n15))
  ) {materialFunction=&MF_NumericDigit}
  n17 = NamedRerouteUsage()
  n18 = ComponentMask(None=@n17) {g=true}
  n19 = AppendVector(A=ComponentMask(None=@n17) {r=true}, B=Min(A=@n18, B=OneMinus(None=@n18)))
  n20 = VectorParameter() {
   desc="R = Scale X, G = Scale Y, B = Offset X, A = Offset Y"
   group=DigitalTimer
   parameterName=Double Colon Separator Scale/Offset
   sortPriority=1
   channelNames {a=Offset Y, b=Offset X, g=Scale Y, r=Scale X}
   defaultValue=<rgba 0.2 0.2 0.75 0>
  }
  n21 = AppendVector(A=@n20.Scale X, B=@n20.Scale X)
  n22 = SphereMask(
   A=Add(
    A=MaterialFunctionCall(UVs=@n19, Texture Scale=@n21) {materialFunction=&ScaleUVsByCenter}.UVs
    B=AppendVector(A=@n20.Offset X, B=@n20.Offset X)
   )
   B=AppendVector(A=Constant() {r=0.5}, B=ScalarParameter() {defaultValue=0.25, group=DigitalTimer, parameterName=Double Colon Spacing})
   Radius=ScalarParameter() {defaultValue=0.05, group=DigitalTimer, parameterName=Double Colon Radius}
  ) {attenuationRadius=0.05}
  n23 = NamedRerouteUsage()
  n24 = VectorParameter() {
   desc="R = Scale X, G = Scale Y, B = Offset X, A = Offset Y"
   group=DigitalTimer
   parameterName=Single Colon Separator Scale/Offset
   sortPriority=1
   channelNames {a=Offset Y, b=Offset X, g=Scale Y, r=Scale X}
   defaultValue=<rgba 0.2 0.2 -0.85 -0.6>
  }
  n25 = AppendVector(A=@n24.Scale X, B=@n24.Scale X)
  n26 = SphereMask(
   A=Add(
    A=MaterialFunctionCall(UVs=@n23, Texture Scale=@n25) {materialFunction=&ScaleUVsByCenter}.UVs
    B=AppendVector(A=@n24.Offset X, B=@n24.Offset X)
   )
   B=Constant2Vector() {g=0.5, r=0.5}
   Radius=ScalarParameter() {defaultValue=0.05, group=DigitalTimer, parameterName=Single Colon Radius}
  ) {attenuationRadius=0.05}
  FunctionOutput(None=Max(A=Max(A=@n10, B=@n16.Result), B=Max(A=@n22, B=@n26)))
  NamedRerouteDeclaration(None=FunctionInput(Preview=TextureCoordinate()) {bUsePreviewValueAsDefault=true, inputName=UVs, inputType=FunctionInput_Vector2}) {name=IN_UVs, nodeColor=<rgba 1 0.6328 0 1>}
  n27 = ScalarParameter() {group=DigitalTimer, parameterName=TimeStart, sortPriority=0}
  n28 = Subtract(A=Time() {period=10}, B=@n27)
  n29 = Saturate(None=@n27)
  NamedRerouteDeclaration(None=StaticSwitchParameter(True=Multiply(A=@n28, B=@n29), False=ScalarParameter() {desc=Time should be in seconds., group=DigitalTimer, parameterName=OverrideTime}) {defaultValue=true, group=DigitalTimer, parameterName=UseTimestamping}) {name=Time, nodeColor=<rgba 0 0.8125 1 1>}
 ]}
