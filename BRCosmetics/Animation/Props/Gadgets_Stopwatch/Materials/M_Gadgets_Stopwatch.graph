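# Stopwatch gadget prop material that draws an amber digital timer screen over the standard 2023 character parent material.
# A UV crop mask blends the screen's base and emissive colors into the PreFX attributes before PostFX, with digits generated by MF_DigitalTimer.
# Side branches hold debug readouts of camera-to-pivot distance and of time, with an optional stepped time whose interval comes from a distance-driven curve atlas.
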Material /BRCosmetics/Animation/Props/Gadgets_Stopwatch/Materials/M_Gadgets_Stopwatch
graph {
 properties {
  bUseMaterialAttributes=true
  bUsedWithMeshDeformer=true
  bUsedWithMorphTargets=true
  bUsedWithSkeletalMesh=true
  pixelDepthOffsetMode=PDOM_Legacy
  shadingModel=MSM_FromMaterialExpression
  usedShadingModels=Default Lit | Subsurface | Clear Coat | Subsurface Profile | Cloth
 }
 nodes [
  n1 = MaterialFunctionCall() {materialFunction=&MF_CharacterParent_2023_PreFX}
  n2 = Constant4Vector() {constant=<rgba 0.16 0.65 0.355 0.75>}
  n3 = MaterialFunctionCall!(
   UVs=TextureCoordinate()
   TextureIn
   UpperLeftCorner=ComponentMask(None=@n2.RGBA) {g=true, r=true}
   LowerRightCorner=ComponentMask(None=@n2.RGBA) {a=true, b=true}
   ExportFloat4
  ) {materialFunction=/Engine/Functions/Engine_MaterialFunctions01/Texturing/TextureCropping}
  n4 = Multiply(A=@n3.Crop Mask, B=Constant4Vector() {constant=<rgba 0.3467 0.0823 0 1>}.RGBA)
  n5 = Multiply(A=NamedRerouteUsage(), B=Constant4Vector() {constant=<rgba 0.0116 0.004 0 1>}.RGBA)
  n6 = NamedRerouteUsage()
  n7 = Multiply(A=@n3.Crop Mask, B=Constant4Vector() {constant=<rgba 1 0.6109 0.1094 1>}.RGBA)
  n8 = NamedRerouteUsage()
  n9 = BlendMaterialAttributes(
   A=@n1.Result
   B=SetMaterialAttributes!(
    MaterialAttributes=@n1.Result
    Base Color=LinearInterpolate(A=@n4, B=@n5, Alpha=@n6)
    Emissive Color=LinearInterpolate(A=@n7, B=@n5, Alpha=@n8)
   ) {attributeSetTypes [#1, #2]}
   Alpha=@n3.Crop Mask
  )
  MaterialFunctionCall(Material=@n9, M=@n1.M, Pre Skinned Local Position=@n1.M, Distance Blend=@n1.M, Fresnel=@n1.M) {materialFunction=&MF_CharacterParent_2023_PostFX}
  n10 = Time() {bOverride_Period=true, period=10}
  n11 = CameraPositionWS()
  n12 = MaterialFunctionCall() {materialFunction=&ObjectPivotPoint}
  n13 = CurveAtlasRowParameter(CurveTime=Divide(A=Distance(A=@n11, B=@n12.Object Pivot Location)) {constB=64}) {
   atlas=/BRCosmetics/Animation/Props/Gadgets_Stopwatch/Textures/CA_Stopwatch_Atlas
   curve=/BRCosmetics/Animation/Props/Gadgets_Stopwatch/Textures/CC_Stopwatch_StepCurve
  }
  n14 = Floor(None=Divide(A=@n10, B=@n13.R) {constB=5})
  n15 = Add(A=Floor(None=@n10), B=Frac(None=Multiply(A=@n14, B=@n13.R) {constB=5}))
  MaterialFunctionCall(
   Number=StaticSwitch(True=NamedRerouteDeclaration(None=@n15) {name=DelayedTime, nodeColor=<rgba 1 0 0.4219 1>}, False=@n10)
   MaximumNumberOfDigits=Constant() {r=3}
   DebugTextLocation RG_UpperRight BA_LowerLeft=Constant4Vector() {constant=<rgba 0.2 0.675 0.375 0.725>}.RGBA
  ) {materialFunction=&DebugScalarValues}
  n16 = MaterialFunctionCall(UVs=TextureCoordinate(), Texture Scale=Constant2Vector() {g=0.3, r=0.2}) {materialFunction=&ScaleUVsByCenter}
  n17 = Constant2Vector() {g=-0.665, r=1.225}
  NamedRerouteDeclaration(None=MaterialFunctionCall!(UVs=Add(A=@n16.UVs, B=@n17)) {materialFunction=&/MF_DigitalTimer}.Result) {name=Digits, nodeColor=<rgba 0 0.3437 1 1>}
  n18 = GetMaterialAttributes(None=@n9) {attributeGetTypes [#2]}
  n19 = VertexInterpolator(VS=Distance(
    A=CameraPositionWS()
    B=MaterialFunctionCall() {materialFunction=&ObjectPivotPoint}.Object Pivot Location
   ))
  n20 = Constant4Vector() {constant=<rgba 0.1 0.6 0.375 0.65>}
  SetMaterialAttributes!(
   MaterialAttributes=@n18.MaterialAttributes
   Emissive Color=Add(
    A=MaterialFunctionCall(Number=@n19.PS, DebugTextLocation RG_UpperRight BA_LowerLeft=@n20.RGBA) {materialFunction=&DebugScalarValues}.Result
    B=@n18.Emissive Color
   )
  ) {attributeSetTypes [#2]}
 ]
}
